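# Physics asset for the dangling accessories on the female medium Baseball Kitbash character body: shoulder mask, two left grenades with pins, and fanny pack zipper.
# Each accessory hangs by limited swing and twist from a master or pack body, with grenade pins chained to their grenades.
# Those anchors, plus left upper arm and right thigh capsules, constrain to the pelvis with free rotation, except the angularly locked mask master.
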PhysicsAsset /BRCosmetics/Characters/Player/Female/Medium/Bodies/F_MED_Baseball_Kitbash/Meshes/F_MED_Baseball_Kitbash_Dyn
bodies {
 dyn_fanny_pack [{shapeType=Capsule, length=1, radius=1}]
 dyn_fanny_zipper [{shapeType=Capsule, length=1.7552, radius=0.9753, center=<1.2578 0.0614 0.0164>, rotation=<rot 0 92.0596 -87.769>}]
 dyn_grenade_bk_l [{shapeType=Capsule, length=0, radius=5, center=<5.8682 0 3.5633>, rotation=<rot 90 0 0>}]
 dyn_grenade_bk_l_MSTR [{shapeType=Capsule, length=1, radius=1}]
 dyn_grenade_bk_pin_l [{shapeType=Capsule, length=0.5, radius=1.5, center=<1.7301 0 0>, rotation=<rot 90 0 0>}]
 dyn_grenade_fr_l [{shapeType=Capsule, length=0, radius=5, center=<2.7521 0 3.5633>, rotation=<rot 90 0.0099 0.0099>}]
 dyn_grenade_fr_l_MSTR [{shapeType=Capsule, length=1, radius=1}]
 dyn_grenade_fr_pin_l [{shapeType=Capsule, length=0.5, radius=1.5, center=<1.7301 0 0>, rotation=<rot 90 0 0>}]
 dyn_shoulder_mask_l [{shapeType=Capsule, length=5.5625, radius=3.6875, center=<8.0485 0.0003 0>, rotation=<rot -0.0001 0 -89.9995>}]
 dyn_shoulder_mask_l_MSTR [{shapeType=Capsule, length=1, radius=1, rotation=<rot 90 0 0>}]
 pelvis [{shapeType=Capsule, length=1, radius=1}]
 thigh_r [{shapeType=Capsule, length=16.8249, radius=8.51, center=<18.6228 -0.0004 0.1225>, rotation=<rot 86.0739 180 180>}]
 upperarm_l [{shapeType=Capsule, length=12, radius=5, center=<9.6306 0 0>, rotation=<rot 90 0 0>}]
}
constraints [
 {bone1Name=dyn_shoulder_mask_l, bone2Name=dyn_shoulder_mask_l_MSTR, swing1LimitDegrees=1, swing2LimitDegrees=45, twistLimitDegrees=2}
 {
  bone1Name=upperarm_l
  bone2Name=pelvis
  swing1LimitDegrees=45
  swing1Motion=Free
  swing2LimitDegrees=45
  swing2Motion=Free
  twistLimitDegrees=45
  twistMotion=Free
 }
 {
  bone1Name=dyn_grenade_fr_l_MSTR
  bone2Name=pelvis
  swing1LimitDegrees=45
  swing1Motion=Free
  swing2LimitDegrees=45
  swing2Motion=Free
  twistLimitDegrees=45
  twistMotion=Free
 }
 {bone1Name=dyn_grenade_fr_l, bone2Name=dyn_grenade_fr_l_MSTR, swing1LimitDegrees=45, swing2LimitDegrees=30, twistLimitDegrees=5}
 {bone1Name=dyn_grenade_fr_pin_l, bone2Name=dyn_grenade_fr_l, swing1LimitDegrees=20, swing2LimitDegrees=45, twistLimitDegrees=10}
 {
  bone1Name=dyn_grenade_bk_l_MSTR
  bone2Name=pelvis
  swing1LimitDegrees=45
  swing1Motion=Free
  swing2LimitDegrees=45
  swing2Motion=Free
  twistLimitDegrees=45
  twistMotion=Free
 }
 {bone1Name=dyn_grenade_bk_l, bone2Name=dyn_grenade_bk_l_MSTR, swing1LimitDegrees=45, swing2LimitDegrees=30, twistLimitDegrees=5}
 {bone1Name=dyn_grenade_bk_pin_l, bone2Name=dyn_grenade_bk_l, swing1LimitDegrees=20, swing2LimitDegrees=45, twistLimitDegrees=10}
 {bone1Name=dyn_fanny_zipper, bone2Name=dyn_fanny_pack, swing1LimitDegrees=45, swing2LimitDegrees=45, twistLimitDegrees=5}
 {
  bone1Name=dyn_fanny_pack
  bone2Name=pelvis
  swing1LimitDegrees=45
  swing1Motion=Free
  swing2LimitDegrees=45
  swing2Motion=Free
  twistLimitDegrees=45
  twistMotion=Free
 }
 {
  bone1Name=thigh_r
  bone2Name=pelvis
  swing1LimitDegrees=45
  swing1Motion=Free
  swing2LimitDegrees=45
  swing2Motion=Free
  twistLimitDegrees=45
  twistMotion=Free
 }
 {
  bone1Name=dyn_shoulder_mask_l_MSTR
  bone2Name=pelvis
  swing1LimitDegrees=45
  swing1Motion=Locked
  swing2LimitDegrees=45
  swing2Motion=Locked
  twistLimitDegrees=45
  twistMotion=Locked
 }
]
