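# Post-pose Anim Blueprint for the female Baseball Kitbash body, adding secondary motion to its shoulder pad, fanny pack and grenade bones.
# It copies the attached parent mesh's pose, then drives the left shoulder pad with constraints and Look At nodes.
# Additive bone offsets correct for clavicle, jump and crouch alphas, with crouch events setting CrouchingAlpha.
# A final RigidBody node simulates the Dyn physics asset relative to spine_01 with damped component motion.
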AnimBlueprint /BRCosmetics/Characters/Player/Female/Medium/Bodies/F_MED_Baseball_Kitbash/Meshes/F_MED_Baseball_Kitbash_AnimBP
target_skeleton=&/F_MED_Baseball_Kitbash_Skeleton
graphs [
 AnimGraph {nodes [
   n1 = "Constraint - dyn_shoulderpad_l_MSTR (PP)"!(
    ComponentPose=Local To Component(LocalPose=Copy Pose From Mesh() {bUseAttachedParent=true})
    Alpha=1
    ConstraintWeights_0=0.7438
    ConstraintWeights_1:float=0.2562
   ) -> (Pose) {
    alphaBoolBlend {blendOption=Linear}
    boneToModify {boneName=dyn_shoulderpad_l_MSTR}
    constraintSetup [{targetBone {boneName=clavicle_l}}, {targetBone {boneName=upperarm_twist_01_l}}]
    constraintWeights [1,1]
   }
   n2 = "Constraint - dyn_shoulderpad_l_MSTR (R)"(ComponentPose=@n1, Alpha=1, ConstraintWeights_0=1) {
    alphaBoolBlend {blendOption=Linear}
    boneToModify {boneName=dyn_shoulderpad_l_MSTR}
    constraintSetup [{transformType=Rotation, targetBone {boneName=clavicle_l}}]
    constraintWeights [1]
   }
   n3 = Look At - Bone: dyn_shoulderpad_l_MSTR(ComponentPose=@n2, Alpha=1) {
    alphaBoolBlend {blendOption=Linear}
    boneToModify {boneName=dyn_shoulderpad_l_MSTR}
    lookAtLocation {x=0}
    lookAtTarget {boneReference {boneName=lowerarm_l}}
    lookAt_Axis {axis=<1 0>}
    lookUp_Axis {axis {x=0, z=1}}
   }
   n4 = Look At - Bone: dyn_shoulder_mask_l_MSTR(ComponentPose=@n3, Alpha=1) {
    bUseLookUpAxis=true
    alphaBoolBlend {blendOption=Linear}
    boneToModify {boneName=dyn_shoulder_mask_l_MSTR}
    lookAtLocation {x=0}
    lookAtTarget {boneReference {boneName=lowerarm_l}}
    lookAt_Axis {axis=<1 0>}
    lookUp_Axis {axis {x=0, z=1}}
   }
   n5 = "Transform (Modify) Bone - Bone: dyn_shoulderpad_l_MSTR"(ComponentPose=@n4, Alpha=Get ClavUpLeft() -> (ClavUpLeft:float), Translation="5.6,-1.697,1.592", Rotation="0,0,0", Scale="1,1,1") {
    rotationMode=BMM_Additive
    rotationSpace=BCS_BoneSpace
    translationMode=BMM_Additive
    translationSpace=BCS_BoneSpace
    alphaBoolBlend {blendOption=Linear}
    alphaScaleBiasClamp {interpSpeedDecreasing=1, interpSpeedIncreasing=1}
    boneToModify {boneName=dyn_shoulderpad_l_MSTR}
    translation=<5.6 -1.697 1.592>
   }
   n6 = "Transform (Modify) Bone - Bone: dyn_shoulderpad_l_MSTR"(ComponentPose=@n5, Alpha=Get ClavDownLeft() -> (ClavDownLeft:float), Translation="3.090,-0.000,1.026", Rotation="0,0,0", Scale="1,1,1") {
    rotationMode=BMM_Additive
    rotationSpace=BCS_BoneSpace
    translationMode=BMM_Additive
    translationSpace=BCS_BoneSpace
    alphaBoolBlend {blendOption=Linear}
    alphaScaleBiasClamp {interpSpeedDecreasing=1, interpSpeedIncreasing=1}
    boneToModify {boneName=dyn_shoulderpad_l_MSTR}
    translation {x=3.09, z=1.026}
   }
   n7 = "Constraint - dyn_fanny_zipper (PPP)"!(ComponentPose=@n6, Alpha=1, ConstraintWeights_0=0.0003, ConstraintWeights_1:float=0.5181, ConstraintWeights_2:float=0.4815) -> (Pose) {
    alphaBoolBlend {blendOption=Linear}
    boneToModify {boneName=dyn_fanny_zipper}
    constraintSetup [{targetBone {boneName=spine_01}}, {targetBone {boneName=dyn_fanny_pack}}, {targetBone {boneName=pelvis}}]
    constraintWeights [1,1,1]
   }
   n8 = "Transform (Modify) Bone - Bone: dyn_grenade_fr_l_MSTR"(
    ComponentPose=@n7
    Alpha=Get JumpAlpha() -> (JumpAlpha:float)
    Translation="-4.478,0.000,-0.000"
    Rotation="50.000,0.000,0.000"
    Scale="1,1,1"
   ) {
    rotationMode=BMM_Additive
    rotationSpace=BCS_BoneSpace
    translationMode=BMM_Additive
    translationSpace=BCS_BoneSpace
    alphaBoolBlend {blendOption=Linear}
    alphaScaleBiasClamp {bInterpResult=true}
    boneToModify {boneName=dyn_grenade_fr_l_MSTR}
    rotation {pitch=49.9999}
    translation {x=-4.478}
   }
   n9 = "Transform (Modify) Bone - Bone: dyn_fanny_pack"(
    ComponentPose=@n8
    Alpha=Get CrouchingAlpha() -> (CrouchingAlpha:double)
    Translation="-0.357,-2.293,-0.560"
    Rotation="4.924,-10.039,-20.873"
    Scale="1,1,1"
   ) {
    rotationMode=BMM_Additive
    rotationSpace=BCS_BoneSpace
    translationMode=BMM_Additive
    translationSpace=BCS_BoneSpace
    alphaBoolBlend {blendOption=Linear}
    alphaScaleBiasClamp {bInterpResult=true}
    boneToModify {boneName=dyn_fanny_pack}
    rotation=<rot 4.924 -10.0394 -20.8727>
    translation=<-0.357 -2.293 -0.56>
   }
   n10 = RigidBody(ComponentPose=@n9) {
    bClampLinearTranslationLimitToRefPose=true
    bOverrideWorldGravity=true
    overridePhysicsAsset=&/F_MED_Baseball_Kitbash_Dyn
    simulationSpace=BaseBoneSpace
    alphaBoolBlend {blendOption=Linear}
    baseBoneRef {boneName=spine_01}
    componentAppliedLinearAccClamp=<3000 3000 3000>
    componentLinearAccScale=<0.45 0.3 0.3>
    componentLinearVelScale=<0.45 0.3 0.3>
    overrideWorldGravity {z=-980}
   }
   Output Pose(Result=Component To Local(ComponentPose=@n10)) {layerGroup=None}
  ]}
 EventGraph {nodes [
   Event OnBeginCrouching()
   > Set CrouchingAlpha(CrouchingAlpha:double=1) -> (Output_Get:double)
   Event OnEndCrouching()
   > Set CrouchingAlpha(CrouchingAlpha=0)
  ]}
]
variables [CrouchingAlpha {type=number}]
